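# Sound cue for the Season 22 battle pass narrative Part 2 quest 01 loop sound.
# A single wave player feeds the loop wave asset straight to the output with no modulation or mixing.
# It plays under the quests sound class and borrows the shared Season 22 quest attenuation from the Part 1 folder, limiting how far it is heard.
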SoundCue /BattlePassS22/QuestActors/Narrative_Pt2/SFX/S22_Pt02_Q01_Loop_Cue
root=WavePlayer_0
nodes [{id=WavePlayer_0, properties {soundWaveAssetPtr=&/S22_Pt02_Q01_Loop}}]
properties {
 attenuationSettings=/BattlePassS22/QuestActors/Narrative_Pt1/SFX/S22_Quests_Atten
 maxDistance=1300
 soundClassObject=&Fort_Quests
 volumeMultiplier=1
}
