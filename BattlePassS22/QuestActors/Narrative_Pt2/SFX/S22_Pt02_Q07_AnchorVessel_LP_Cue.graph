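# Looping quest sound cue for the Anchor Vessel in Season 22 battle pass narrative part 2, quest 7.
# A Modulator applies a very slight random pitch variation to a single looping Wave Player.
# It plays spatialized through the quest's dedicated attenuation asset and routes to the Fort_Quests sound class.
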SoundCue /BattlePassS22/QuestActors/Narrative_Pt2/SFX/S22_Pt02_Q07_AnchorVessel_LP_Cue
root=Modulator_0
nodes [
 {id=Modulator_0, children [WavePlayer_1], properties {pitchMax=1.01, pitchMin=0.99}}
 {id=WavePlayer_1, properties {bLooping=true, soundWaveAssetPtr=&/S22_Pt02_Q07_AnchorVessel_LP}}
]
properties {
 attenuationSettings=&/S22_Pt02_Q07_Atten
 maxDistance=1100
 soundClassObject=&Fort_Quests
 volumeMultiplier=1
}
