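# Sound cue for the completion of the Anchor Vessel step in Season 22 narrative quest Part 2, Q09.
# A random node picks one of three equally weighted completion waves and feeds it through a modulator that adds slight pitch variation at constant volume.
# It plays in the quests sound class and uses the shared Season 22 quest attenuation from the Part 1 folder.
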SoundCue /BattlePassS22/QuestActors/Narrative_Pt2/SFX/S22_Pt02_Q09_AnchorVessel_Complete_Cue
root=Modulator_0
nodes [
 {id=Modulator_0, children [Random_0], properties {pitchMax=1.01, pitchMin=0.98, volumeMax=1, volumeMin=1}}
 {id=Random_0, children [WavePlayer_0,WavePlayer_1,WavePlayer_2], properties {weights [1,1,1]}}
 {id=WavePlayer_0, properties {soundWaveAssetPtr=&/S22_Pt02_Q09_AnchorVessel_Complete_01}}
 {id=WavePlayer_1, properties {soundWaveAssetPtr=&/S22_Pt02_Q09_AnchorVessel_Complete_02}}
 {id=WavePlayer_2, properties {soundWaveAssetPtr=&/S22_Pt02_Q09_AnchorVessel_Complete_03}}
]
properties {
 attenuationSettings=/BattlePassS22/QuestActors/Narrative_Pt1/SFX/S22_Quests_Atten
 maxDistance=1300
 soundClassObject=&Fort_Quests
}
