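# Looping sound cue for the anchor vessel in quest 9 of the Season 22 Battle Pass narrative, part 2.
# A looping wave player feeds a Modulator left at default settings.
# The cue plays in the quest sound class at reduced volume and borrows the Q07 attenuation asset instead of one specific to Q09.
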SoundCue /BattlePassS22/QuestActors/Narrative_Pt2/SFX/S22_Pt02_Q09_AnchorVessel_LP_Cue
root=Modulator_0
nodes [
 {id=Modulator_0, children [WavePlayer_0]}
 {id=WavePlayer_0, properties {bLooping=true, soundWaveAssetPtr=&/S22_Pt02_Q09_AnchorVessel_LP}}
]
properties {
 attenuationSettings=&/S22_Pt02_Q07_Atten
 maxDistance=1100
 soundClassObject=&Fort_Quests
 volumeMultiplier=0.5
}
